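# Editor utility widget that sets the PCG editing mode on actors selected in the level.
# PreConstruct binds a single property view to the widget's In Editing Mode variable, exposing the mode choice.
# The Swap Edit Mode button loops over selected actors and applies that mode, as both current and serialized, to each one's valid PCGComponent.
# An unconnected literal string node holds pasted MeshGrass actor data and plays no part.
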EditorUtilityWidgetBlueprint /FigmentTools/PCGTools/EUW_GrassPCG_TypeSwap
graphs [EventGraph {nodes [
   "On Clicked (EditorUtilityButton_48)"() {is=title K2Node_ComponentBoundEvent}
   > n1 = GetSelectedLevelActors(self=Get EditorActorSubsystem())
   > n2 = For Each Loop(Array:Array of Actor Object References=@n1) -> (Array Element:Actor Object Reference)
   n3 = GetComponentByClass!(self:Actor Object Reference=@n2, ComponentClass:Actor Component Class Reference=/Script/PCG.PCGComponent) -> (ReturnValue:PCGComponent Object Reference)
   > Is Valid(InputObject=@n3)
   n4 = Get In Editing Mode() -> (In Editing Mode:EPCGEditorDirtyMode Enum=Normal)
   > SetEditingMode!(
    execute
    self:PCGComponent Object Reference=@n3
    InEditingMode:EPCGEditorDirtyMode Enum=@n4
    InSerializedEditingMode:EPCGEditorDirtyMode Enum=@n4
   ) -> (then)
   Event PreConstruct()
   n5 = Get SinglePropertyView_67() -> (SinglePropertyView_67:Single Property View Object Reference)
   > SetObject(self=@n5, NewObject=Self-Reference())
   > SetPropertyName!(execute, self:Single Property View Object Reference=@n5, NewPropertyName:Name=In Editing Mode) -> (then)
   MakeLiteralString(Value="Begin Map\r\n   Begin Level\r\n      Begin Actor Class=/Game/Athena/Helios/Environments/Landscape/WorldBuildingTools/PCGGrass/BP_MeshGrass_GlobalControls.BP_MeshGrass_GlobalControls_C Name=BP_MeshGrass_GlobalControls_C_UAID_124247F586253D3702_1106995313 Archetype=\"&BP_MeshGrass_GlobalControls_C'/Game/Athena/Helios/Environments/Landscape/WorldBuildingTools/PCGGrass/BP_MeshGrass_GlobalControls.Default__BP_MeshGrass_GlobalControls_C'\" ActorFolderPath=\"MeshGrass\" ExportPath=\"&BP_MeshGrass_GlobalControls_C'/Figment_S08_Map/Athena_Terrain_S08:PersistentLevel.BP_MeshGrass_GlobalControls_C_UAID_124247F586253D3702_1106995313'\"\r\n         Begin Object Class=/Script/Engine.BillboardComponent Name=\"Billboard\" Archetype=\"&BillboardComponent'&BP_MeshGrass_GlobalControls_C:Billboard_GEN_VARIABLE'\" ExportPath=\"&BillboardComponent'&Athena_Terrain_S08:PersistentLevel.BP_MeshGrass_GlobalControls_C_UAID_124247F586253D3702_1106995313.Billboard'\"\r\n         End Object\r\n         Begin Object Class=/Game/Athena/Helios/Environments/Landscape/WorldBuildingTools/PCGGrass/BP_MeshGrassIOComponent.BP_MeshGrassIOComponent_C Name=\"BP_MeshGrassIOComponent\" Archetype=\"&BP_MeshGrassIOComponent_C'&BP_MeshGrass_GlobalControls_C:BP_MeshGrassIOComponent_GEN_VARIABLE'\" ExportPath=\"&BP_MeshGrassIOComponent_C'&Athena_Terrain_S08:PersistentLevel.BP_MeshGrass_GlobalControls_C_UAID_124247F586253D3702_1106995313.BP_MeshGrassIOComponent'\"\r\n         End Object\r\n         Begin Object Name=\"Billboard\" ExportPath=\"&BillboardComponent'&Athena_Terrain_S08:PersistentLevel.BP_MeshGrass_GlobalControls_C_UAID_124247F586253D3702_1106995313.Billboard'\"\r\n            RelativeLocation=(X=84758.4684,Y=-18232.6191,Z=1917.4443)\r\n            UCSSerializationIndex=0\r\n            bNetAddressable=True\r\n            CreationMethod=SimpleConstructionScript\r\n         End Object\r\n         Begin Object Name=\"BP_MeshGrassIOComponent\" ExportPath=\"&BP_MeshGrassIOComponent_C'&Athena_Terrain_S08:PersistentLevel.BP_MeshGrass_GlobalControls_C_UAID_124247F586253D3702_1106995313.BP_MeshGrassIOComponent'\"\r\n            UCSSerializationIndex=0\r\n            bNetAddressable=True\r\n            CreationMethod=SimpleConstructionScript\r\n         End Object\r\n         BP_MeshGrassIOComponent=\"BP_MeshGrassIOComponent\"\r\n         Billboard=\"Billboard\"\r\n         BiomeTexture=\"/Script/Engine.Texture2D'/FigmentEnvironment/LM_8_1_1/Athena/Environments/Landscape/Textures/T_Athena_Terrain_Biome_M_Tiled_AridFix'\"\r\n         FallbackBiome=NewEnumerator0\r\n         MaxPointsPerMesh=500000\r\n         BiomeBase_BiomeSetup=(BiomeTexturePage_38_#1=-1)\r\n         BiomeBase_MeshSetupVariety=(GrassType_14_#2=\"/Script/Landscape.LandscapeGrassType'/Figment_S08/Athena/Environments/Landscape/Foliage/Grass/Figment_Foliage_GrassOnly'\",VarietiesPreviewWeighting_13_#3=((MeshRead-Only_64_#4=\"&StaticMesh'/FigmentEnvironment/Foliage/Athena_SM_WavyGrass_TopologyColor'\",MaterialOverrideRead-Only_66_#5=\"/Script/Engine.MaterialInstanceConstant'/Figment_S08/Athena/Environments/Landscape/Foliage/Grass/Materials/PCG_GrassMaterial/MI_Athena_Grass_TopoColor'\",Weight_57_#6=1)))\r\n         BiomeBase_MeshSetupBasic=(MeshSetup_12_#7=(()))\r\n         ExcludeByMaterialName=True\r\n         MaterialExclusionStrings=\"ForceMountainDirt\"\r\n         BiomeBase=(Active_15_#4=True,BiomeTexture_54_#8=\"&Texture2D'&T_Athena_Terrain_Biome_M_Tiled_AridFix'\",BiomeCoordinates_51_#9=(X=-1,Y=-1),AllowExclusions_43_#10=True,VarietyData_62_#11=((GrassMesh_5_#12=\"&StaticMesh'&Athena_SM_WavyGrass_TopologyColor'\",OverrideMaterial_15_#13=\"&MaterialInstanceConstant'&MI_Athena_Grass_TopoColor'\",Weight_8_#14=1,EndCullDistance_23_#15=4500,WPODisableDistance_26_#16=0,ScalingMode_50_#17=NewEnumerator2,ScaleMax_31_#18=(X=1.3,Y=1.3,Z=1.3),DetailMode_37_#19=1,SDPTags_53_#20=\"SDP-0-0-0\")),MeshData_61_#21=((SDPTags_90_#22=\"SDP-0-0-0\")),ObjectDependencies_71_#23=(\"&Figment_Foliage_GrassOnly\"))\r\n         BiomeA_BiomeSetup=(Enabled_20_#2=True,BiomeTextureChannel_43_#24=NewEnumerator3)\r\n         BiomeA_MeshSetupVariety=(GrassType_14_#2=\"&LandscapeGrassType'/Figment_S08/Athena/Environments/Landscape/Foliage/Grass/Athena_Foliage_Arid_Grass_Only'\",VarietiesPreviewWeighting_13_#3=((MeshRead-Only_64_#4=\"&StaticMesh'/Figment_S08/Migrated_From_LM_7_00/Athena/Environments/Landscape/Foliage/SM_Athena_DesertGrass01'\",Weight_57_#6=1)))\r\n         BiomeA=(Active_15_#4=True,BiomeTexture_54_#8=\"&Texture2D'&T_Athena_Terrain_Biome_M_Tiled_AridFix'\",BiomeCoordinates_51_#9=(X=0,Y=2),AllowExclusions_43_#10=True,VarietyData_62_#11=((GrassMesh_5_#12=\"&StaticMesh'&SM_Athena_DesertGrass01'\",Weight_8_#14=1,StartCullDistance_21_#25=10000,EndCullDistance_23_#15=10000,WPODisableDistance_26_#16=0,ScalingMode_50_#17=NewEnumerator2,ScaleMax_31_#18=(X=3,Y=3,Z=5),DetailMode_37_#19=1,SDPTags_53_#20=\"SDP-0-0-0\")),ObjectDependencies_71_#23=(\"&Athena_Foliage_Arid_Grass_Only\"))\r\n         BiomeB=(BiomeTexture_54_#8=\"&Texture2D'&T_Athena_Terrain_Biome_M_Tiled_AridFix'\",BiomeCoordinates_51_#9=(X=0,Y=-1),AllowExclusions_43_#10=True)\r\n         BiomeC=(BiomeTexture_54_#8=\"&Texture2D'&T_Athena_Terrain_Biome_M_Tiled_AridFix'\",BiomeCoordinates_51_#9=(X=0,Y=-1),AllowExclusions_43_#10=True)\r\n         BiomeD=(BiomeTexture_54_#8=\"&Texture2D'&T_Athena_Terrain_Biome_M_Tiled_AridFix'\",BiomeCoordinates_51_#9=(X=0,Y=-1),AllowExclusions_43_#10=True)\r\n         BiomeE=(BiomeTexture_54_#8=\"&Texture2D'&T_Athena_Terrain_Biome_M_Tiled_AridFix'\",BiomeCoordinates_51_#9=(X=0,Y=-1),AllowExclusions_43_#10=True)\r\n         BiomeF=(BiomeTexture_54_#8=\"&Texture2D'&T_Athena_Terrain_Biome_M_Tiled_AridFix'\",BiomeCoordinates_51_#9=(X=0,Y=-1),AllowExclusions_43_#10=True)\r\n         BiomeG=(BiomeTexture_54_#8=\"&Texture2D'&T_Athena_Terrain_Biome_M_Tiled_AridFix'\",BiomeCoordinates_51_#9=(X=0,Y=-1),AllowExclusions_43_#10=True)\r\n         UseCustomBiomeTextureProjectionSize=True\r\n         CustomBiomeTextureProjectionSize=306000\r\n         RootComponent=\"Billboard\"\r\n         ActorLabel=\"BP_MeshGrass_GlobalControls\"\r\n      End Actor\r\n      Begin Actor Class=/Game/Athena/Helios/Environments/Landscape/WorldBuildingTools/PCGGrass/BP_MeshGrass_Partitioner.BP_MeshGrass_Partitioner_C Name=BP_MeshGrass_Partitioner_C_UAID_124247F586253D3702_1106998314 Archetype=\"&BP_MeshGrass_Partitioner_C'/Game/Athena/Helios/Environments/Landscape/WorldBuildingTools/PCGGrass/BP_MeshGrass_Partitioner.Default__BP_MeshGrass_Partitioner_C'\" ActorFolderPath=\"MeshGrass\" ExportPath=\"&BP_MeshGrass_Partitioner_C'&Athena_Terrain_S08:PersistentLevel.BP_MeshGrass_Partitioner_C_UAID_124247F586253D3702_1106998314'\"\r\n         Begin Object Class=&PCGComponent Name=\"PCG_MeshGrass_Partitioner\" Archetype=\"&PCGComponent'&BP_MeshGrass_Partitioner_C:PCG_MeshGrass_Partitioner_GEN_VARIABLE'\" ExportPath=\"&PCGComponent'&Athena_Terrain_S08:PersistentLevel.BP_MeshGrass_Partitioner_C_UAID_124247F586253D3702_1106998314.PCG_MeshGrass_Partitioner'\"\r\n            Begin Object Class=/Script/PCG.PCGGraphInstance Name=\"PCGGraphInstance\" Archetype=\"&PCGGraphInstance'&BP_MeshGrass_Partitioner_C:PCG_MeshGrass_Partitioner_GEN_VARIABLE.PCGGraphInstance'\" ExportPath=\"&PCGGraphInstance'&Athena_Terrain_S08:PersistentLevel.BP_MeshGrass_Partitioner_C_UAID_124247F586253D3702_1106998314.PCG_MeshGrass_Partitioner.PCGGraphInstance'\"\r\n            End Object\r\n         End Object\r\n         Begin Object Class=/Script/Engine.BoxComponent Name=\"Volume\" Archetype=\"&BoxComponent'&BP_MeshGrass_Partitioner_C:Volume_GEN_VARIABLE'\" ExportPath=\"&BoxComponent'&Athena_Terrain_S08:PersistentLevel.BP_MeshGrass_Partitioner_C_UAID_124247F586253D3702_1106998314.Volume'\"\r\n            Begin Object Class=/Script/Engine.BodySetup Name=\"BodySetup_0\" Archetype=\"&BodySetup'&BP_MeshGrass_Partitioner_C:Volume_GEN_VARIABLE.BodySetup_0'\" ExportPath=\"&BodySetup'&Athena_Terrain_S08:PersistentLevel.BP_MeshGrass_Partitioner_C_UAID_124247F586253D3702_1106998314.Volume.BodySetup_0'\"\r\n            End Object\r\n         End Object\r\n         Begin Object Class=&PCGComponent Name=\"PCG_MeshGrass_ScalabilityManager\" Archetype=\"&PCGComponent'&BP_MeshGrass_Partitioner_C:PCG_MeshGrass_ScalabilityManager_GEN_VARIABLE'\" ExportPath=\"&PCGComponent'&Athena_Terrain_S08:PersistentLevel.BP_MeshGrass_Partitioner_C_UAID_124247F586253D3702_1106998314.PCG_MeshGrass_ScalabilityManager'\"\r\n            Begin Object Class=&PCGGraphInstance Name=\"PCGGraphInstance\" Archetype=\"&PCGGraphInstance'&BP_MeshGrass_Partitioner_C:PCG_MeshGrass_ScalabilityManager_GEN_VARIABLE.PCGGraphInstance'\" ExportPath=\"&PCGGraphInstance'&Athena_Terrain_S08:PersistentLevel.BP_MeshGrass_Partitioner_C_UAID_124247F586253D3702_1106998314.PCG_MeshGrass_ScalabilityManager.PCGGraphInstance'\"\r\n            End Object\r\n         End Object\r\n         Begin Object Name=\"PCG_MeshGrass_Partitioner\" ExportPath=\"&PCGComponent'&Athena_Terrain_S08:PersistentLevel.BP_MeshGrass_Partitioner_C_UAID_124247F586253D3702_1106998314.PCG_MeshGrass_Partitioner'\"\r\n            Begin Object Name=\"PCGGraphInstance\" ExportPath=\"&PCGGraphInstance'&Athena_Terrain_S08:PersistentLevel.BP_MeshGrass_Partitioner_C_UAID_124247F586253D3702_1106998314.PCG_MeshGrass_Partitioner.PCGGraphInstance'\"\r\n            End Object\r\n            SchedulingPolicy=None\r\n            bForceGenerateOnBPAddedToWorld=False\r\n            bGenerated=True\r\n            GraphInstance=\"PCGGraphInstance\"\r\n            LastGeneratedBounds=(Min=(X=-152000,Y=-170000,Z=-50000),Max=(X=148000,Y=130000,Z=50000),IsValid=True)\r\n            UCSSerializationIndex=0\r\n            bNetAddressable=True\r\n            CreationMethod=SimpleConstructionScript\r\n         End Object\r\n         Begin Object Name=\"Volume\" ExportPath=\"&BoxComponent'&Athena_Terrain_S08:PersistentLevel.BP_MeshGrass_Partitioner_C_UAID_124247F586253D3702_1106998314.Volume'\"\r\n            Begin Object Name=\"BodySetup_0\" ExportPath=\"&BodySetup'&Athena_Terrain_S08:PersistentLevel.BP_MeshGrass_Partitioner_C_UAID_124247F586253D3702_1106998314.Volume.BodySetup_0'\"\r\n               AggGeom=(BoxElems=((X=300000,Y=300000,Z=100000)))\r\n            End Object\r\n            BoxExtent=(X=150000,Y=150000,Z=50000)\r\n            RelativeLocation=(X=-2000,Y=-20000,Z=0)\r\n            UCSSerializationIndex=0\r\n            bNetAddressable=True\r\n            CreationMethod=SimpleConstructionScript\r\n         End Object\r\n         Begin Object Name=\"PCG_MeshGrass_ScalabilityManager\" ExportPath=\"&PCGComponent'&Athena_Terrain_S08:PersistentLevel.BP_MeshGrass_Partitioner_C_UAID_124247F586253D3702_1106998314.PCG_MeshGrass_ScalabilityManager'\"\r\n            Begin Object Name=\"PCGGraphInstance\" ExportPath=\"&PCGGraphInstance'&Athena_Terrain_S08:PersistentLevel.BP_MeshGrass_Partitioner_C_UAID_124247F586253D3702_1106998314.PCG_MeshGrass_ScalabilityManager.PCGGraphInstance'\"\r\n            End Object\r\n            bForceGenerateOnBPAddedToWorld=False\r\n            bGenerated=True\r\n            GraphInstance=\"PCGGraphInstance\"\r\n            LastGeneratedBounds=(Min=(X=-152000,Y=-170000,Z=-50000),Max=(X=148000,Y=130000,Z=50000),IsValid=True)\r\n            UCSSerializationIndex=0\r\n            bNetAddressable=True\r\n            CreationMethod=SimpleConstructionScript\r\n         End Object\r\n         PCG_MeshGrass_ScalabilityManager=\"PCG_MeshGrass_ScalabilityManager\"\r\n         Volume=\"Volume\"\r\n         PCG_MeshGrass_Partitioner=\"PCG_MeshGrass_Partitioner\"\r\n         RootComponent=\"Volume\"\r\n         ActorLabel=\"BP_MeshGrass_Partitioner\"\r\n      End Actor\r\n   End Level\r\nBegin Surface\r\nEnd Surface\r\nEnd Map\r\n")
   Event Construct()
   Event Tick()
  ]}]
variables [In Editing Mode {type=EPCGEditorDirtyMode, default=Normal}]
widgets {
 info {rootWidgetClass=&Overlay, widgetCount=5}
 widgets [
  Overlay_0 {widgetClassPath=&Overlay, widget=Overlay_0}
  VerticalBox_31 {
   widgetClassPath=&VerticalBox
   parent=Overlay_0
   widget=VerticalBox_31
   slot {class=OverlaySlot, properties {padding=<ltrb 4 4 4 4>}}
  }
  SinglePropertyView_67 {
   widgetClassPath=&SinglePropertyView
   parent=VerticalBox_31
   widget=SinglePropertyView_67
   slot {class=VerticalBoxSlot}
  }
  EditorUtilityButton_48 {
   widgetClassPath=&EditorUtilityButton
   parent=VerticalBox_31
   widget=EditorUtilityButton_48
   slot {class=VerticalBoxSlot, properties {padding=<ltrb 0 4 0 0>}}
  }
  TextBlock_59 {
   widgetClassPath=&TextBlock
   parent=EditorUtilityButton_48
   widget=TextBlock_59
   properties {
    text=Swap Edit Mode
    font {
     bForceMonospaced=false
     bMaterialIsStencil=false
     fontMaterial=null
     fontObject=&Roboto
     letterSpacing=0
     monospacedWidth=1
     size=12
     skewAmount=0
     typefaceFontName=Bold
     outlineSettings {
      bApplyOutlineToDropShadows=false
      bMiteredCorners=false
      bSeparateFillAlpha=false
      outlineMaterial=null
      outlineSize=0
      outlineColor=<rgba 0 0 0 1>
     }
    }
   }
   slot {class=ButtonSlot}
  }
 ]
}
